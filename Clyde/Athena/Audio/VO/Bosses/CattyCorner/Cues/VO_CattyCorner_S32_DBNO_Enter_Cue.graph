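# Sound cue for the CattyCorner boss's voice line on entering the DBNO (down but not out) state.
# A random node picks one of three equally weighted voice recordings.
# It plays on a critical enemy VO sound class with shared alerted-NPC attenuation, a slight volume boost, and a boss combat-chatter concurrency set that governs simultaneous playback.
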SoundCue /Clyde/Athena/Audio/VO/Bosses/CattyCorner/Cues/VO_CattyCorner_S32_DBNO_Enter_Cue
root=Random_0
nodes [
 {id=Random_0, children [WavePlayer_0,WavePlayer_1,WavePlayer_2], properties {weights [1,1,1]}}
 {id=WavePlayer_0, properties {soundWaveAssetPtr=/Clyde/Athena/Audio/VO/Bosses/CattyCorner/Waves/VO_CattyCorner_S32_DBNO_Enter_010}}
 {id=WavePlayer_1, properties {soundWaveAssetPtr=/Clyde/Athena/Audio/VO/Bosses/CattyCorner/Waves/VO_CattyCorner_S32_DBNO_Enter_011}}
 {id=WavePlayer_2, properties {soundWaveAssetPtr=/Clyde/Athena/Audio/VO/Bosses/CattyCorner/Waves/VO_CattyCorner_S32_DBNO_Enter_012}}
]
properties {
 attenuationSettings=/Game/Sounds/Fort_Human_VO_BR/NPCs/Shared/NPC_VO_Alerted_Att
 maxDistance=12500
 soundClassObject=&Fort_VO_Enemy_Critical
 volumeMultiplier=1.1147
 concurrencySet [&NPC_Boss_Combat_Chatter_Conc]
}
