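# Sound cue for idle patrol voice lines of the CattyCorner boss, likely from Season 32 given the S32 naming.
# A Random node picks with equal weight from ten wave players, mixing five generic idle takes with five patrol-specific takes.
# Playback uses shared close-range NPC VO attenuation, the critical enemy VO sound class and the boss combat chatter concurrency set, at reduced volume.
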SoundCue /Clyde/Athena/Audio/VO/Bosses/CattyCorner/Cues/VO_CattyCorner_S32_IdlePatrol_Cue
root=Random_0
nodes [
 {
  id=Random_0
  children [WavePlayer_0,WavePlayer_1,WavePlayer_2,WavePlayer_3,WavePlayer_4,WavePlayer_5,WavePlayer_6,WavePlayer_7,WavePlayer_8,WavePlayer_9]
  properties {weights [1,1,1,1,1,1,1,1,1,1]}
 }
 {id=WavePlayer_0, properties {soundWaveAssetPtr=/Clyde/Athena/Audio/VO/Bosses/CattyCorner/Waves/VO_CattyCorner_S32_IdleGeneric_010}}
 {id=WavePlayer_1, properties {soundWaveAssetPtr=/Clyde/Athena/Audio/VO/Bosses/CattyCorner/Waves/VO_CattyCorner_S32_IdleGeneric_020}}
 {id=WavePlayer_2, properties {soundWaveAssetPtr=/Clyde/Athena/Audio/VO/Bosses/CattyCorner/Waves/VO_CattyCorner_S32_IdleGeneric_040}}
 {id=WavePlayer_3, properties {soundWaveAssetPtr=/Clyde/Athena/Audio/VO/Bosses/CattyCorner/Waves/VO_CattyCorner_S32_IdleGeneric_041}}
 {id=WavePlayer_4, properties {soundWaveAssetPtr=/Clyde/Athena/Audio/VO/Bosses/CattyCorner/Waves/VO_CattyCorner_S32_IdleGeneric_050}}
 {id=WavePlayer_5, properties {soundWaveAssetPtr=/Clyde/Athena/Audio/VO/Bosses/CattyCorner/Waves/VO_CattyCorner_S32_IdlePatrol_010}}
 {id=WavePlayer_6, properties {soundWaveAssetPtr=/Clyde/Athena/Audio/VO/Bosses/CattyCorner/Waves/VO_CattyCorner_S32_IdlePatrol_011}}
 {id=WavePlayer_7, properties {soundWaveAssetPtr=/Clyde/Athena/Audio/VO/Bosses/CattyCorner/Waves/VO_CattyCorner_S32_IdlePatrol_020}}
 {id=WavePlayer_8, properties {soundWaveAssetPtr=/Clyde/Athena/Audio/VO/Bosses/CattyCorner/Waves/VO_CattyCorner_S32_IdlePatrol_040}}
 {id=WavePlayer_9, properties {soundWaveAssetPtr=/Clyde/Athena/Audio/VO/Bosses/CattyCorner/Waves/VO_CattyCorner_S32_IdlePatrol_050}}
]
properties {
 attenuationSettings=/Game/Sounds/Fort_Human_VO_BR/NPCs/Shared/NPC_VO_Close_Att
 maxDistance=8500
 soundClassObject=&Fort_VO_Enemy_Critical
 volumeMultiplier=0.4743
 concurrencySet [&NPC_Boss_Combat_Chatter_Conc]
}
